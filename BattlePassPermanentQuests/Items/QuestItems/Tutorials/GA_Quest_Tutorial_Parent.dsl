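; Parent gameplay ability for quest tutorials that grants an urgent transient quest to the owning player in Battle Royale playlists.
; On activation it checks for the required game phase, otherwise waits for the bus skydive landing event, then delays before granting the configured quest.
; When that quest completes, a custom stat event with the tutorial source tag completes the associated wrapper quest.
; The completion binding is removed when the ability ends.
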(event Ability|EventFailedToActivatePassiveAbility
  (params (ActorInfo :type Gameplay%20Ability%20Actor%20Info%20Structure))
  (|CallParentFunction))

(event Ability|EventActivateAbility
  ;[0 0 0] If the player is already in the gameplay phase, grant the urgent quest after a small delay; if not wait for landing from the bus to grant the next quest
  (|CallParentFunction)
  (if (Class|FortGameStateAthena|IsBattleRoyalePlaylist (Fortnite|Athena|GameState|GetGameStateAthenafromContext (Variables|GAVars|Data|GetOwningPlayerPawn)) false)
    (if (Utilities|Enum|Equal(Enum) (Variables|Default|GetRequiredGamePhase) (BattleRoyaleGamePhaseLogic|GetGamePhase (BattleRoyaleGamePhaseLogic|GetBattleRoyaleGamePhaseLogicComponent)))
      (Development|PrintString "Already in gameplay phase, starting delay for tutorial..." true true "(R=1.000000,G=0.000000,B=0.222520,A=1.000000)" 3.0)
      (else
        (Ability|Tasks|WaitGameplayEvent "(TagName=\"Event.SkydiveEnded.FromBus\")" 0 true
          (:EventReceived
            (Development|PrintString "Landed from bus, starting delay for tutorial..." true true "(R=1.000000,G=0.000000,B=0.222520,A=1.000000)" 3.0)
            (Utilities|FlowControl|Delay 4.0)))))
    (Utilities|FlowControl|Delay 5.0)
    (bind _asfort_player_controller_athena (Utilities|Casting|CastToFortPlayerControllerAthena (Fortnite|GetFortPlayerControllerfromActor (Variables|GAVars|Data|GetOwningPlayerPawn))))
    (Class|FortControllerComponentTransientQuests|GrantTransientQuest (Class|FortPlayerControllerAthena|GetTransientQuestsComponent _asfort_player_controller_athena) (Variables|Default|GetUrgentQuestToGrant) 0)
    (Quests|BindEventtoOnUrgentTransientQuestCompleted (Class|FortQuestManager|GetQuestManagerComponent (Variables|GAVars|Data|GetOwningPlayerQuestManager) "/Script/FortniteGame.FortQuestManagerComponent_Athena") (EventDispatchers|CreateEvent :!event Custom|SendQuestUpdate))))

(event Custom|SendQuestUpdate
  (params (TransientQuestItem :type Fort%20Quest%20Item%20Object%20Reference))
  ;[0 0 0] When an urgent tutorial quest is completed, send a stat update to complete the associated wrapper quest
  (if (== (AssetManager|GetPrimaryAssetIdfromObject (Quest|GetQuestDefinition TransientQuestItem)) (AssetManager|GetPrimaryAssetIdfromObject (Variables|Default|GetUrgentQuestToGrant)))
    (Quest|SendComplexCustomStatEvent (Variables|GAVars|Data|GetOwningPlayerQuestManager) 0 (Variables|Default|GetTutorialQuestSourceTag) "(GameplayTags=)" 1)))

(event Ability|EventOnEndAbility
  (params (bWasCancelled :type Boolean))
  (|CallParentFunction bWasCancelled)
  (Quests|UnbindEventfromOnUrgentTransientQuestCompleted (Class|FortQuestManager|GetQuestManagerComponent (Variables|GAVars|Data|GetOwningPlayerQuestManager) "/Script/FortniteGame.FortQuestManagerComponent_Athena") (EventDispatchers|CreateEvent :!event Custom|SendQuestUpdate)))
